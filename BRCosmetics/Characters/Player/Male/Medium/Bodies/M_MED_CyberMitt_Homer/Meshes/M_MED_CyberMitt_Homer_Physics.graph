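# Physics asset for the male medium CyberMitt Homer cosmetic body, covering only the two hoodie laces and the upper spine.
# Each lace is a chain of small capsules, a base and four dyn bones, linked by constraints with narrow swing and very tight twist limits.
# Both lace bases attach to a large spine_05 capsule, the only torso body, through constraints with free swing and twist.
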PhysicsAsset /BRCosmetics/Characters/Player/Male/Medium/Bodies/M_MED_CyberMitt_Homer/Meshes/M_MED_CyberMitt_Homer_Physics
bodies {
 Base_Hoodie_lace_01_l [{shapeType=Capsule, length=2, radius=1, shapeName=Base_Hoodie_lace_01_l_capsule, center=<2 0 0>, rotation=<rot 90 0 0>}]
 Base_Hoodie_lace_01_r [{shapeType=Capsule, length=0.505, radius=0.505, shapeName=Base_Hoodie_lace_01_r_capsule, rotation=<rot 0 -90 0>}]
 dyn_Hoodie_lace_01_l [{shapeType=Capsule, length=2, radius=1, shapeName=dyn_Hoodie_lace_01_l_capsule, center=<2 0 0>, rotation=<rot 90 0 0>}]
 dyn_Hoodie_lace_01_r [{shapeType=Capsule, length=2, radius=1, shapeName=dyn_Hoodie_lace_01_r_capsule, center=<2 0 0>, rotation=<rot 90 0 0>}]
 dyn_Hoodie_lace_02_l [{shapeType=Capsule, length=2, radius=1, shapeName=dyn_Hoodie_lace_02_l_capsule, center=<2 0 0>, rotation=<rot 90 0 0>}]
 dyn_Hoodie_lace_02_r [{shapeType=Capsule, length=2, radius=1, shapeName=dyn_Hoodie_lace_02_r_capsule, center=<2 0 0>, rotation=<rot 90 0 0>}]
 dyn_Hoodie_lace_03_l [{shapeType=Capsule, length=2, radius=1, shapeName=dyn_Hoodie_lace_03_l_capsule, center=<2 0 0>, rotation=<rot 90 0 0>}]
 dyn_Hoodie_lace_03_r [{shapeType=Capsule, length=2, radius=1, shapeName=dyn_Hoodie_lace_03_r_capsule, center=<2 0 0>, rotation=<rot 90 0 0>}]
 dyn_Hoodie_lace_04_l [{shapeType=Capsule, length=2, radius=1, shapeName=dyn_Hoodie_lace_04_l_capsule, center=<2 0 0>, rotation=<rot 90 0 0>}]
 dyn_Hoodie_lace_04_r [{shapeType=Capsule, length=2, radius=1, shapeName=dyn_Hoodie_lace_04_r_capsule, center=<2 0 0>, rotation=<rot 90 0 0>}]
 spine_05 [{
   shapeType=Capsule
   length=12.5389
   radius=11.8311
   shapeName=spine_05_capsule
   center=<-3.7102 1.5019 0.9085>
   rotation=<rot 1.737 -173.5848 1.4601>
  }]
}
constraints [
 {bone1Name=dyn_Hoodie_lace_01_l, bone2Name=Base_Hoodie_lace_01_l, swing1LimitDegrees=15, swing2LimitDegrees=15, twistLimitDegrees=1}
 {bone1Name=dyn_Hoodie_lace_02_l, bone2Name=dyn_Hoodie_lace_01_l, swing1LimitDegrees=15, swing2LimitDegrees=15, twistLimitDegrees=1}
 {bone1Name=dyn_Hoodie_lace_03_l, bone2Name=dyn_Hoodie_lace_02_l, swing1LimitDegrees=15, swing2LimitDegrees=15, twistLimitDegrees=1}
 {bone1Name=dyn_Hoodie_lace_04_l, bone2Name=dyn_Hoodie_lace_03_l, swing1LimitDegrees=15, swing2LimitDegrees=15, twistLimitDegrees=1}
 {bone1Name=dyn_Hoodie_lace_01_r, bone2Name=Base_Hoodie_lace_01_r, swing1LimitDegrees=15, swing2LimitDegrees=15, twistLimitDegrees=1}
 {bone1Name=dyn_Hoodie_lace_02_r, bone2Name=dyn_Hoodie_lace_01_r, swing1LimitDegrees=15, swing2LimitDegrees=15, twistLimitDegrees=1}
 {bone1Name=dyn_Hoodie_lace_03_r, bone2Name=dyn_Hoodie_lace_02_r, swing1LimitDegrees=15, swing2LimitDegrees=15, twistLimitDegrees=1}
 {bone1Name=dyn_Hoodie_lace_04_r, bone2Name=dyn_Hoodie_lace_03_r, swing1LimitDegrees=15, swing2LimitDegrees=15, twistLimitDegrees=1}
 {
  bone1Name=Base_Hoodie_lace_01_l
  bone2Name=spine_05
  swing1LimitDegrees=45
  swing1Motion=Free
  swing2LimitDegrees=45
  swing2Motion=Free
  twistLimitDegrees=45
  twistMotion=Free
 }
 {
  bone1Name=Base_Hoodie_lace_01_r
  bone2Name=spine_05
  swing1LimitDegrees=45
  swing1Motion=Free
  swing2LimitDegrees=45
  swing2Motion=Free
  twistLimitDegrees=45
  twistMotion=Free
 }
]
