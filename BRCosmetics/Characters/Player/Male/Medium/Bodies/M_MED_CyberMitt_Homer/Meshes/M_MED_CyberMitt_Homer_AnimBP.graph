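# Post-process-style cosmetic AnimBP for the CyberMitt Homer body, adding secondary motion to clothing and accessories.
# Copies the parent mesh pose through a Control Rig fed a helmet flag, weighted bone constraints, RigidBody, then chained AnimDynamics for laces, straps, belts, zipper, pockets and hood.
# Updates set sim-space and gravity via a deprecated state helper, with front-end overrides for alphas and simulation space.
# On customization load, the cosmetic's Stage2 variant sets the helmet flag.
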AnimBlueprint /BRCosmetics/Characters/Player/Male/Medium/Bodies/M_MED_CyberMitt_Homer/Meshes/M_MED_CyberMitt_Homer_AnimBP
target_skeleton=&/M_MED_CyberMitt_Homer_Skeleton
graphs [
 AnimGraph {nodes [
   n1 = Control Rig!(
    Source=Copy Pose From Mesh() {bUseAttachedParent=true}
    Alpha
    IsUsingHelmet:bool=Get isUsingHelmet() -> (isUsingHelmet:bool)
    "isCrouching ":bool=true
   ) -> (Pose) {
    alphaBoolBlend {blendOption=Linear}
    controlRigAssetReference {blueprintRigClass=&/M_MED_CyberMitt_Homer_CtrlRig.M_MED_CyberMitt_Homer_CtrlRig_C}
   }
   n2 = "Constraint - Base_Shoe_Strap_r (PPP)"!(
    ComponentPose=Local To Component(LocalPose=@n1)
    Alpha=1
    ConstraintWeights_0=0.7478
    ConstraintWeights_1:float=0.165
    ConstraintWeights_2:float=0.0871
   ) -> (Pose) {
    alphaBoolBlend {blendOption=Linear}
    boneToModify {boneName=Base_Shoe_Strap_r}
    constraintSetup [{targetBone {boneName=calf_twist_01_r}}, {targetBone {boneName=foot_r}}, {targetBone {boneName=calf_twist_02_r}}]
    constraintWeights [1,1,1]
   }
   n3 = "Constraint - Base_Shoe_Strap_l (PPP)"!(ComponentPose=@n2, Alpha=1, ConstraintWeights_0=0.7486, ConstraintWeights_1:float=0.1649, ConstraintWeights_2:float=0.0866) -> (Pose) {
    alphaBoolBlend {blendOption=Linear}
    boneToModify {boneName=Base_Shoe_Strap_l}
    constraintSetup [{targetBone {boneName=calf_twist_01_l}}, {targetBone {boneName=foot_l}}, {targetBone {boneName=calf_twist_02_l}}]
    constraintWeights [1,1,1]
   }
   n4 = "Constraint - Base_upper_pocket_r (PPPPP)"!(
    ComponentPose=@n3
    Alpha=1
    ConstraintWeights_0=0.7907
    ConstraintWeights_1:float=0.0956
    ConstraintWeights_2:float=0.0635
    ConstraintWeights_3:float=0.0473
    ConstraintWeights_4:float=0.003
   ) -> (Pose) {
    alphaBoolBlend {blendOption=Linear}
    boneToModify {boneName=Base_upper_pocket_r}
    constraintSetup [
     {targetBone {boneName=thigh_twist_01_r}}
     {targetBone {boneName=thigh_r}}
     {targetBone {boneName=pelvis}}
     {targetBone {boneName=deform_groin_r}}
     {targetBone {boneName=deform_glute_r}}
    ]
    constraintWeights [1,1,1,1,1]
   }
   n5 = "Constraint - Base_lower_pocket_r (PPPPPPP)"!(
    ComponentPose=@n4
    Alpha=1
    ConstraintWeights_0=0.4722
    ConstraintWeights_1:float=0.4692
    ConstraintWeights_2:float=0.0302
    ConstraintWeights_3:float=0.0194
    ConstraintWeights_4:float=0.0082
    ConstraintWeights_5:float=0.0007
    ConstraintWeights_6:float=0.0001
   ) -> (Pose) {
    alphaBoolBlend {blendOption=Linear}
    boneToModify {boneName=Base_lower_pocket_r}
    constraintSetup [
     {targetBone {boneName=thigh_twist_01_r}}
     {targetBone {boneName=thigh_r}}
     {targetBone {boneName=calf_r}}
     {targetBone {boneName=deform_knee_out_r}}
     {targetBone {boneName=deform_knee_in_r}}
     {targetBone {boneName=calf_twist_02_r}}
     {targetBone {boneName=deform_groin_r}}
    ]
    constraintWeights [1,1,1,1,1,1,1]
   }
   n6 = "Constraint - Base_upper_pocket_l (PPPPP)"!(
    ComponentPose=@n5
    Alpha=1
    ConstraintWeights_0=0.7905
    ConstraintWeights_1:float=0.0874
    ConstraintWeights_2:float=0.0691
    ConstraintWeights_3:float=0.0417
    ConstraintWeights_4:float=0.0112
   ) -> (Pose) {
    alphaBoolBlend {blendOption=Linear}
    boneToModify {boneName=Base_upper_pocket_l}
    constraintSetup [
     {targetBone {boneName=thigh_twist_01_l}}
     {targetBone {boneName=thigh_l}}
     {targetBone {boneName=pelvis}}
     {targetBone {boneName=deform_groin_l}}
     {targetBone {boneName=deform_glute_l}}
    ]
    constraintWeights [1,1,1,1,1]
   }
   n7 = "Constraint - Base_lower_pocket_l (PPPPPP)"!(
    ComponentPose=@n6
    Alpha=1
    ConstraintWeights_0=0.4863
    ConstraintWeights_1:float=0.4552
    ConstraintWeights_2:float=0.0306
    ConstraintWeights_3:float=0.0172
    ConstraintWeights_4:float=0.0096
    ConstraintWeights_5:float=0.001
   ) -> (Pose) {
    alphaBoolBlend {blendOption=Linear}
    boneToModify {boneName=Base_lower_pocket_l}
    constraintSetup [
     {targetBone {boneName=thigh_l}}
     {targetBone {boneName=thigh_twist_01_l}}
     {targetBone {boneName=calf_l}}
     {targetBone {boneName=deform_knee_out_l}}
     {targetBone {boneName=deform_knee_in_l}}
     {targetBone {boneName=calf_twist_02_l}}
    ]
    constraintWeights [1,1,1,1,1,1]
   }
   n8 = "Constraint - Base_Zipper_m (PPPPP)"!(
    ComponentPose=@n7
    Alpha=1
    ConstraintWeights_0=0.585
    ConstraintWeights_1:float=0.1652
    ConstraintWeights_2:float=0.1515
    ConstraintWeights_3:float=0.0918
    ConstraintWeights_4:float=0.0066
   ) -> (Pose) {
    alphaBoolBlend {blendOption=Linear}
    boneToModify {boneName=Base_Zipper_m}
    constraintSetup [
     {targetBone {boneName=pelvis}}
     {targetBone {boneName=spine_01}}
     {targetBone {boneName=spine_02}}
     {targetBone {boneName=spine_03}}
     {targetBone {boneName=spine_04}}
    ]
    constraintWeights [1,1,1,1,1]
   }
   n9 = "Constraint - Base_Shoe_lace_r_01_r1 (PPP)"!(ComponentPose=@n8, Alpha=1, ConstraintWeights_0=0.5411, ConstraintWeights_1:float=0.4035, ConstraintWeights_2:float=0.0554) -> (Pose) {
    alphaBoolBlend {blendOption=Linear}
    boneToModify {boneName=Base_Shoe_lace_r_01_r1}
    constraintSetup [{targetBone {boneName=foot_r}}, {targetBone {boneName=calf_twist_01_r}}, {targetBone {boneName=calf_twist_02_r}}]
    constraintWeights [1,1,1]
   }
   n10 = "Constraint - Base_Shoe_lace_r_01_l (PPP)"!(ComponentPose=@n9, Alpha=1, ConstraintWeights_0=0.5896, ConstraintWeights_1:float=0.3644, ConstraintWeights_2:float=0.046) -> (Pose) {
    alphaBoolBlend {blendOption=Linear}
    boneToModify {boneName=Base_Shoe_lace_r_01_l}
    constraintSetup [{targetBone {boneName=foot_r}}, {targetBone {boneName=calf_twist_01_r}}, {targetBone {boneName=calf_twist_02_r}}]
    constraintWeights [1,1,1]
   }
   n11 = "Constraint - Base_Shoe_lace_l_01_l (PPP)"!(ComponentPose=@n10, Alpha=1, ConstraintWeights_0=0.5896, ConstraintWeights_1:float=0.3644, ConstraintWeights_2:float=0.046) -> (Pose) {
    alphaBoolBlend {blendOption=Linear}
    boneToModify {boneName=Base_Shoe_lace_l_01_l}
    constraintSetup [{targetBone {boneName=foot_l}}, {targetBone {boneName=calf_twist_01_l}}, {targetBone {boneName=calf_twist_02_l}}]
    constraintWeights [1,1,1]
   }
   n12 = "Constraint - Base_Shoe_lace_l_01_r (PPP)"!(ComponentPose=@n11, Alpha=1, ConstraintWeights_0=0.5544, ConstraintWeights_1:float=0.3876, ConstraintWeights_2:float=0.058) -> (Pose) {
    alphaBoolBlend {blendOption=Linear}
    boneToModify {boneName=Base_Shoe_lace_l_01_r}
    constraintSetup [{targetBone {boneName=foot_l}}, {targetBone {boneName=calf_twist_01_l}}, {targetBone {boneName=calf_twist_02_l}}]
    constraintWeights [1,1,1]
   }
   n13 = "Constraint - Base_Hoodie_lace_01_l (PPPPPPPP)"!(
    ComponentPose=@n12
    Alpha=1
    ConstraintWeights_0=0.8388
    ConstraintWeights_1:float=0.0691
    ConstraintWeights_2:float=0.0266
    ConstraintWeights_3:float=0.0266
    ConstraintWeights_4:float=0.0186
    ConstraintWeights_5:float=0.0108
    ConstraintWeights_6:float=0.0047
    ConstraintWeights_7:float=0.0046
   ) -> (Pose) {
    alphaBoolBlend {blendOption=Linear}
    boneToModify {boneName=Base_Hoodie_lace_01_l}
    constraintSetup [
     {targetBone {boneName=spine_05}}
     {targetBone {boneName=spine_04}}
     {targetBone {boneName=clavicle_l}}
     {targetBone {boneName=hood_front_l}}
     {targetBone {boneName=spine_03}}
     {targetBone {boneName=neck_02}}
     {targetBone {boneName=spine_02}}
     {targetBone {boneName=neck_01}}
    ]
    constraintWeights [1,1,1,1,1,1,1,1]
   }
   n14 = "Constraint - Base_Hoodie_lace_01_r (PPPPPPPP)"!(
    ComponentPose=@n13
    Alpha=1
    ConstraintWeights_0=0.6014
    ConstraintWeights_1:float=0.2165
    ConstraintWeights_2:float=0.0707
    ConstraintWeights_3:float=0.0452
    ConstraintWeights_4:float=0.031
    ConstraintWeights_5:float=0.0183
    ConstraintWeights_6:float=0.0129
    ConstraintWeights_7:float=0.004
   ) -> (Pose) {
    alphaBoolBlend {blendOption=Linear}
    boneToModify {boneName=Base_Hoodie_lace_01_r}
    constraintSetup [
     {targetBone {boneName=spine_05}}
     {targetBone {boneName=hood_front_r}}
     {targetBone {boneName=spine_04}}
     {targetBone {boneName=clavicle_r}}
     {targetBone {boneName=neck_02}}
     {targetBone {boneName=spine_03}}
     {targetBone {boneName=neck_01}}
     {targetBone {boneName=deform_pec_r}}
    ]
    constraintWeights [1,1,1,1,1,1,1,1]
   }
   n15 = RigidBody!(
    ComponentPose=@n14
    Alpha
    OverrideWorldGravity:Vector="0,0,0"
    SimSpaceSettings:Sim Space Settings Structure=(WorldAlpha=0.35,VelocityScaleZ=0.2,DampingAlpha=1,MaxLinearVelocity=600,MaxAngularVelocity=9,MaxLinearAcceleration=2500,MaxAngularAcceleration=38,ExternalLinearDragV=(X=0.1,Y=0.1,Z=0.1),ExternalLinearVelocity=(X=0,Y=0,Z=0),ExternalAngularVelocity=(X=0,Y=0,Z=0))
   ) -> (Pose) {
    bClampLinearTranslationLimitToRefPose=true
    bOverrideWorldGravity=true
    bUseExternalClothCollision=true
    overridePhysicsAsset=&/M_MED_CyberMitt_Homer_Physics
    simulationSpace=BaseBoneSpace
    alphaBoolBlend {blendOption=Linear}
    baseBoneRef {boneName=spine_04}
    componentAppliedLinearAccClamp=<800 800 800>
    componentLinearAccScale=<0.3 0.3 0.2>
    componentLinearVelScale=<0.3 0.3 0.2>
    simSpaceSettings {
     maxAngularAcceleration=38
     maxAngularVelocity=9
     maxLinearAcceleration=2500
     maxLinearVelocity=600
     velocityScaleZ=0.2
     worldAlpha=0.35
     externalLinearDragV=<0.1 0.1 0.1>
    }
   }
   n16 = Anim Dynamics - Bone: dyn_Shoe_lace_r_01_r!(ComponentPose=@n15, Alpha=1, GravityOverride:Vector="0,0,-980", ExternalForce="0,0,0") -> (Pose) {
    angularDampingOverride=1
    angularSpringConstant=20000
    bAngularSpring=true
    bOverrideAngularDamping=true
    bUseGravityOverride=true
    gravityScale=0.35
    simulationSpace=BoneRelative
    alphaBoolBlend {blendOption=Linear}
    boundBone {boneName=dyn_Shoe_lace_r_01_r}
    componentLinearAccScale=<1 1 1>
    componentLinearVelScale=<1 1 1>
    gravityOverride {z=-980}
    physicsBodyDefinitions [{
      boundBone {boneName=dyn_Shoe_lace_r_01_r}
      boxExtents=<20 30 30>
      constraintSetup {angularLimitsMax=<0 0 30>, angularLimitsMin=<0 -40 -30>, angularTarget=<1 0 0>}
      localJointOffset=<4 0 0>
     }]
    relativeSpaceBone {boneName=foot_r}
   }
   n17 = Anim Dynamics - Bone: dyn_Shoe_lace_r_02_r!(ComponentPose=@n16, Alpha=1, GravityOverride:Vector="0,0,-980", ExternalForce="0,0,0") -> (Pose) {
    angularDampingOverride=1
    angularSpringConstant=20000
    bAngularSpring=true
    bOverrideAngularDamping=true
    bUseGravityOverride=true
    gravityScale=0.35
    simulationSpace=BoneRelative
    alphaBoolBlend {blendOption=Linear}
    boundBone {boneName=dyn_Shoe_lace_r_02_r}
    componentLinearAccScale=<1 1 1>
    componentLinearVelScale=<1 1 1>
    gravityOverride {z=-980}
    physicsBodyDefinitions [{
      boundBone {boneName=dyn_Shoe_lace_r_02_r}
      boxExtents=<20 30 30>
      constraintSetup {angularLimitsMax=<0 0 30>, angularLimitsMin=<0 -30 -30>, angularTarget=<1 0 0>}
      localJointOffset=<4 0 0>
     }]
    relativeSpaceBone {boneName=foot_r}
   }
   n18 = Anim Dynamics - Bone: dyn_Shoe_lace_r_03_r!(ComponentPose=@n17, Alpha=1, GravityOverride:Vector="0,0,-980", ExternalForce="0,0,0") -> (Pose) {
    angularDampingOverride=1
    angularSpringConstant=20000
    bAngularSpring=true
    bOverrideAngularDamping=true
    bUseGravityOverride=true
    gravityScale=0.35
    simulationSpace=BoneRelative
    alphaBoolBlend {blendOption=Linear}
    boundBone {boneName=dyn_Shoe_lace_r_03_r}
    componentLinearAccScale=<1 1 1>
    componentLinearVelScale=<1 1 1>
    gravityOverride {z=-980}
    physicsBodyDefinitions [{
      boundBone {boneName=dyn_Shoe_lace_r_03_r}
      boxExtents=<20 30 30>
      constraintSetup {angularLimitsMax=<0 0 30>, angularLimitsMin=<0 -30 -30>, angularTarget=<1 0 0>}
      localJointOffset=<4 0 0>
     }]
    relativeSpaceBone {boneName=foot_r}
   }
   n19 = Anim Dynamics - Bone: dyn_Shoe_lace_r_01_l!(ComponentPose=@n18, Alpha=1, GravityOverride:Vector="0,0,-980", ExternalForce="0,0,0") -> (Pose) {
    angularDampingOverride=1
    angularSpringConstant=20000
    bAngularSpring=true
    bOverrideAngularDamping=true
    bUseGravityOverride=true
    gravityScale=0.35
    simulationSpace=BoneRelative
    alphaBoolBlend {blendOption=Linear}
    boundBone {boneName=dyn_Shoe_lace_r_01_l}
    componentLinearAccScale=<1 1 1>
    componentLinearVelScale=<1 1 1>
    gravityOverride {z=-980}
    physicsBodyDefinitions [{
      boundBone {boneName=dyn_Shoe_lace_r_01_l}
      boxExtents=<20 30 30>
      constraintSetup {angularLimitsMax=<0 0 30>, angularLimitsMin=<0 -30 -30>, angularTarget=<1 0 0>}
      localJointOffset=<4 0 0>
     }]
    relativeSpaceBone {boneName=foot_r}
   }
   n20 = Anim Dynamics - Bone: dyn_Shoe_lace_r_02_l!(ComponentPose=@n19, Alpha=1, GravityOverride:Vector="0,0,-980", ExternalForce="0,0,0") -> (Pose) {
    angularDampingOverride=1
    angularSpringConstant=20000
    bAngularSpring=true
    bOverrideAngularDamping=true
    bUseGravityOverride=true
    gravityScale=0.35
    simulationSpace=BoneRelative
    alphaBoolBlend {blendOption=Linear}
    boundBone {boneName=dyn_Shoe_lace_r_02_l}
    componentLinearAccScale=<1 1 1>
    componentLinearVelScale=<1 1 1>
    gravityOverride {z=-980}
    physicsBodyDefinitions [{
      boundBone {boneName=dyn_Shoe_lace_r_02_l}
      boxExtents=<20 30 30>
      constraintSetup {angularLimitsMax=<0 0 30>, angularLimitsMin=<0 -30 -30>, angularTarget=<1 0 0>}
      localJointOffset=<4 0 0>
     }]
    relativeSpaceBone {boneName=foot_r}
   }
   n21 = Anim Dynamics - Bone: dyn_Shoe_lace_r_03_l!(ComponentPose=@n20, Alpha=1, GravityOverride:Vector="0,0,-980", ExternalForce="0,0,0") -> (Pose) {
    angularDampingOverride=1
    angularSpringConstant=20000
    bAngularSpring=true
    bOverrideAngularDamping=true
    bUseGravityOverride=true
    gravityScale=0.35
    simulationSpace=BoneRelative
    alphaBoolBlend {blendOption=Linear}
    boundBone {boneName=dyn_Shoe_lace_r_03_l}
    componentLinearAccScale=<1 1 1>
    componentLinearVelScale=<1 1 1>
    gravityOverride {z=-980}
    physicsBodyDefinitions [{
      boundBone {boneName=dyn_Shoe_lace_r_03_l}
      boxExtents=<20 30 30>
      constraintSetup {angularLimitsMax=<0 0 30>, angularLimitsMin=<0 -30 -30>, angularTarget=<1 0 0>}
      localJointOffset=<4 0 0>
     }]
    relativeSpaceBone {boneName=foot_r}
   }
   n22 = Anim Dynamics - Bone: dyn_Shoe_lace_l_01_r!(ComponentPose=@n21, Alpha=1, GravityOverride:Vector="0,0,-980", ExternalForce="0,0,0") -> (Pose) {
    angularDampingOverride=1
    angularSpringConstant=20000
    bAngularSpring=true
    bOverrideAngularDamping=true
    bUseGravityOverride=true
    gravityScale=0.35
    simulationSpace=BoneRelative
    alphaBoolBlend {blendOption=Linear}
    boundBone {boneName=dyn_Shoe_lace_l_01_r}
    componentLinearAccScale=<1 1 1>
    componentLinearVelScale=<1 1 1>
    gravityOverride {z=-980}
    physicsBodyDefinitions [{
      boundBone {boneName=dyn_Shoe_lace_l_01_r}
      boxExtents=<20 30 30>
      constraintSetup {angularLimitsMax=<0 0 30>, angularLimitsMin=<0 -30 -30>, angularTarget=<1 0 0>}
      localJointOffset=<4 0 0>
     }]
    relativeSpaceBone {boneName=foot_l}
   }
   n23 = Anim Dynamics - Bone: dyn_Shoe_lace_l_02_r!(ComponentPose=@n22, Alpha=1, GravityOverride:Vector="0,0,-980", ExternalForce="0,0,0") -> (Pose) {
    angularDampingOverride=1
    angularSpringConstant=20000
    bAngularSpring=true
    bOverrideAngularDamping=true
    bUseGravityOverride=true
    gravityScale=0.35
    simulationSpace=BoneRelative
    alphaBoolBlend {blendOption=Linear}
    boundBone {boneName=dyn_Shoe_lace_l_02_r}
    componentLinearAccScale=<1 1 1>
    componentLinearVelScale=<1 1 1>
    gravityOverride {z=-980}
    physicsBodyDefinitions [{
      boundBone {boneName=dyn_Shoe_lace_l_02_r}
      boxExtents=<20 30 30>
      constraintSetup {angularLimitsMax=<0 0 30>, angularLimitsMin=<0 -30 -30>, angularTarget=<1 0 0>}
      localJointOffset=<4 0 0>
     }]
    relativeSpaceBone {boneName=foot_l}
   }
   n24 = Anim Dynamics - Bone: dyn_Shoe_lace_l_03_r!(ComponentPose=@n23, Alpha=1, GravityOverride:Vector="0,0,-980", ExternalForce="0,0,0") -> (Pose) {
    angularDampingOverride=1
    angularSpringConstant=20000
    bAngularSpring=true
    bOverrideAngularDamping=true
    bUseGravityOverride=true
    gravityScale=0.35
    simulationSpace=BoneRelative
    alphaBoolBlend {blendOption=Linear}
    boundBone {boneName=dyn_Shoe_lace_l_03_r}
    componentLinearAccScale=<1 1 1>
    componentLinearVelScale=<1 1 1>
    gravityOverride {z=-980}
    physicsBodyDefinitions [{
      boundBone {boneName=dyn_Shoe_lace_l_03_r}
      boxExtents=<20 30 30>
      constraintSetup {angularLimitsMax=<0 0 30>, angularLimitsMin=<0 -30 -30>, angularTarget=<1 0 0>}
      localJointOffset=<4 0 0>
     }]
    relativeSpaceBone {boneName=foot_l}
   }
   n25 = Anim Dynamics - Bone: dyn_Shoe_lace_l_01_l!(ComponentPose=@n24, Alpha=1, GravityOverride:Vector="0,0,-980", ExternalForce="0,0,0") -> (Pose) {
    angularDampingOverride=1
    angularSpringConstant=20000
    bAngularSpring=true
    bOverrideAngularDamping=true
    bUseGravityOverride=true
    gravityScale=0.35
    simulationSpace=BoneRelative
    alphaBoolBlend {blendOption=Linear}
    boundBone {boneName=dyn_Shoe_lace_l_01_l}
    componentLinearAccScale=<1 1 1>
    componentLinearVelScale=<1 1 1>
    gravityOverride {z=-980}
    physicsBodyDefinitions [{
      boundBone {boneName=dyn_Shoe_lace_l_01_l}
      boxExtents=<20 30 30>
      constraintSetup {angularLimitsMax=<0 0 30>, angularLimitsMin=<0 -30 -30>, angularTarget=<1 0 0>}
      localJointOffset=<4 0 0>
     }]
    relativeSpaceBone {boneName=foot_l}
   }
   n26 = Anim Dynamics - Bone: dyn_Shoe_lace_l_02_l!(ComponentPose=@n25, Alpha=1, GravityOverride:Vector="0,0,-980", ExternalForce="0,0,0") -> (Pose) {
    angularDampingOverride=1
    angularSpringConstant=20000
    bAngularSpring=true
    bOverrideAngularDamping=true
    bUseGravityOverride=true
    gravityScale=0.35
    simulationSpace=BoneRelative
    alphaBoolBlend {blendOption=Linear}
    boundBone {boneName=dyn_Shoe_lace_l_02_l}
    componentLinearAccScale=<1 1 1>
    componentLinearVelScale=<1 1 1>
    gravityOverride {z=-980}
    physicsBodyDefinitions [{
      boundBone {boneName=dyn_Shoe_lace_l_02_l}
      boxExtents=<20 30 30>
      constraintSetup {angularLimitsMax=<0 0 30>, angularLimitsMin=<0 -30 -30>, angularTarget=<1 0 0>}
      localJointOffset=<4 0 0>
     }]
    relativeSpaceBone {boneName=foot_l}
   }
   n27 = Anim Dynamics - Bone: dyn_Shoe_lace_l_03_l!(ComponentPose=@n26, Alpha=1, GravityOverride:Vector="0,0,-980", ExternalForce="0,0,0") -> (Pose) {
    angularDampingOverride=1
    angularSpringConstant=20000
    bAngularSpring=true
    bOverrideAngularDamping=true
    bUseGravityOverride=true
    gravityScale=0.35
    simulationSpace=BoneRelative
    alphaBoolBlend {blendOption=Linear}
    boundBone {boneName=dyn_Shoe_lace_l_03_l}
    componentLinearAccScale=<1 1 1>
    componentLinearVelScale=<1 1 1>
    gravityOverride {z=-980}
    physicsBodyDefinitions [{
      boundBone {boneName=dyn_Shoe_lace_l_03_l}
      boxExtents=<20 30 30>
      constraintSetup {angularLimitsMax=<0 0 30>, angularLimitsMin=<0 -30 -30>, angularTarget=<1 0 0>}
      localJointOffset=<4 0 0>
     }]
    relativeSpaceBone {boneName=foot_l}
   }
   n28 = Anim Dynamics - Bone: dyn_lower_belt_fnt_l!(ComponentPose=@n27, Alpha=1, GravityOverride:Vector="0,0,-980", ExternalForce="0,0,0") -> (Pose) {
    angularDampingOverride=0.99
    angularSpringConstant=500
    bAngularSpring=true
    bOverrideAngularDamping=true
    bUseGravityOverride=true
    gravityScale=0.35
    simulationSpace=RootRelative
    alphaBoolBlend {blendOption=Linear}
    boundBone {boneName=dyn_lower_belt_fnt_l}
    componentAppliedLinearAccClamp=<3000 3000 3000>
    componentLinearAccScale=<0.4 0.4 0.4>
    componentLinearVelScale=<0.4 0.4 0.4>
    gravityOverride {z=-980}
    physicsBodyDefinitions [{
      boundBone {boneName=dyn_lower_belt_fnt_l}
      boxExtents=<20 30 30>
      constraintSetup {angularLimitsMax=<0 4 4>, angularLimitsMin=<0 -50 -5>, angularTarget=<1 0 0>}
      localJointOffset=<4 0 0>
     }]
    relativeSpaceBone {boneName=thigh_twist_01_l}
   }
   n29 = Anim Dynamics - Bone: dyn_lower_belt_fnt_r!(ComponentPose=@n28, Alpha=1, GravityOverride:Vector="0,0,-980", ExternalForce="0,0,0") -> (Pose) {
    angularDampingOverride=0.99
    angularSpringConstant=500
    bAngularSpring=true
    bOverrideAngularDamping=true
    bUseGravityOverride=true
    gravityScale=0.35
    simulationSpace=RootRelative
    alphaBoolBlend {blendOption=Linear}
    boundBone {boneName=dyn_lower_belt_fnt_r}
    componentAppliedLinearAccClamp=<3000 3000 3000>
    componentLinearAccScale=<0.4 0.4 0.4>
    componentLinearVelScale=<0.4 0.4 0.4>
    gravityOverride {z=-980}
    physicsBodyDefinitions [{
      boundBone {boneName=dyn_lower_belt_fnt_r}
      boxExtents=<20 30 30>
      constraintSetup {angularLimitsMax=<0 3 0>, angularLimitsMin=<0 -50 -3>, angularTarget=<1 0 0>}
      localJointOffset=<4 0 0>
     }]
    relativeSpaceBone {boneName=thigh_twist_01_r}
   }
   n30 = Anim Dynamics - Bone: dyn_upper_belt_r!(ComponentPose=@n29, Alpha=1, GravityOverride:Vector="0,0,-980", ExternalForce="0,0,0") -> (Pose) {
    angularDampingOverride=0.99
    angularSpringConstant=500
    bAngularSpring=true
    bOverrideAngularDamping=true
    bUseGravityOverride=true
    gravityScale=0.35
    simulationSpace=RootRelative
    alphaBoolBlend {blendOption=Linear}
    boundBone {boneName=dyn_upper_belt_r}
    componentAppliedLinearAccClamp=<3000 3000 3000>
    componentLinearAccScale=<0.4 0.4 0.4>
    componentLinearVelScale=<0.4 0.4 0.4>
    gravityOverride {z=-980}
    physicsBodyDefinitions [{
      boundBone {boneName=dyn_upper_belt_r}
      boxExtents=<20 30 30>
      constraintSetup {angularLimitsMax=<0 0 15>, angularLimitsMin=<0 -15 -15>, angularTarget=<1 0 0>}
      localJointOffset=<4 0 0>
     }]
    relativeSpaceBone {boneName=thigh_r}
   }
   n31 = Anim Dynamics - Bone: dyn_upper_belt_l!(ComponentPose=@n30, Alpha=1, GravityOverride:Vector="0,0,-980", ExternalForce="0,0,0") -> (Pose) {
    angularDampingOverride=0.99
    angularSpringConstant=500
    bAngularSpring=true
    bOverrideAngularDamping=true
    bUseGravityOverride=true
    gravityScale=0.35
    simulationSpace=RootRelative
    alphaBoolBlend {blendOption=Linear}
    boundBone {boneName=dyn_upper_belt_l}
    componentAppliedLinearAccClamp=<3000 3000 3000>
    componentLinearAccScale=<0.4 0.4 0.4>
    componentLinearVelScale=<0.4 0.4 0.4>
    gravityOverride {z=-980}
    physicsBodyDefinitions [{
      boundBone {boneName=dyn_upper_belt_l}
      boxExtents=<20 30 30>
      constraintSetup {angularLimitsMax=<0 0 5>, angularLimitsMin=<0 -35 -5>, angularTarget=<1 0 0>}
      localJointOffset=<4 0 0>
     }]
    relativeSpaceBone {boneName=thigh_l}
   }
   n32 = Anim Dynamics - Bone: dyn_Zipper_m!(ComponentPose=@n31, Alpha=1, GravityOverride:Vector="0,0,-980", ExternalForce="0,0,0") -> (Pose) {
    angularDampingOverride=0.99
    angularSpringConstant=1000
    bAngularSpring=true
    bOverrideAngularDamping=true
    bUseGravityOverride=true
    gravityScale=0.35
    simulationSpace=RootRelative
    alphaBoolBlend {blendOption=Linear}
    boundBone {boneName=dyn_Zipper_m}
    componentAppliedLinearAccClamp=<3000 3000 3000>
    componentLinearAccScale=<0.6 0.6 0.6>
    componentLinearVelScale=<0.6 0.6 0.6>
    gravityOverride {z=-980}
    physicsBodyDefinitions [{
      boundBone {boneName=dyn_Zipper_m}
      boxExtents=<20 15 15>
      constraintSetup {angularLimitsMax=<0 30 0>, angularLimitsMin=<0 -30 -30>, angularTarget=<1 0 0>}
      localJointOffset=<4 0 0>
     }]
   }
   n33 = Anim Dynamics - Bone: dyn_upper_pocket_l!(ComponentPose=@n32, Alpha=1, GravityOverride:Vector="0,0,-980", ExternalForce="0,0,0") -> (Pose) {
    angularDampingOverride=1
    angularSpringConstant=20000
    bAngularSpring=true
    bOverrideAngularDamping=true
    bUseGravityOverride=true
    gravityScale=0.35
    simulationSpace=RootRelative
    alphaBoolBlend {blendOption=Linear}
    boundBone {boneName=dyn_upper_pocket_l}
    componentAppliedLinearAccClamp=<3000 3000 3000>
    componentLinearAccScale=<1 1 1>
    componentLinearVelScale=<1 1 1>
    gravityOverride {z=-980}
    physicsBodyDefinitions [{
      boundBone {boneName=dyn_upper_pocket_l}
      boxExtents=<15 25 25>
      constraintSetup {angularLimitsMax=<0 0 15>, angularLimitsMin=<0 0 -50>, angularTarget=<1 0 0>}
      localJointOffset=<4 0 0>
     }]
   }
   n34 = Anim Dynamics - Bone: dyn_lower_pocket_l!(ComponentPose=@n33, Alpha=1, GravityOverride:Vector="0,0,-980", ExternalForce="0,0,0") -> (Pose) {
    angularDampingOverride=1
    angularSpringConstant=20000
    bAngularSpring=true
    bOverrideAngularDamping=true
    bUseGravityOverride=true
    gravityScale=0.35
    simulationSpace=RootRelative
    alphaBoolBlend {blendOption=Linear}
    boundBone {boneName=dyn_lower_pocket_l}
    componentAppliedLinearAccClamp=<3000 3000 3000>
    componentLinearAccScale=<1 1 1>
    componentLinearVelScale=<1 1 1>
    gravityOverride {z=-980}
    physicsBodyDefinitions [{
      boundBone {boneName=dyn_lower_pocket_l}
      boxExtents=<15 25 25>
      constraintSetup {angularLimitsMax=<0 0 15>, angularLimitsMin=<0 0 -50>, angularTarget=<1 0 0>}
      localJointOffset=<4 0 0>
     }]
   }
   n35 = Anim Dynamics - Bone: dyn_lower_pocket_r!(ComponentPose=@n34, Alpha=1, GravityOverride:Vector="0,0,-980", ExternalForce="0,0,0") -> (Pose) {
    angularDampingOverride=1
    angularSpringConstant=20000
    bAngularSpring=true
    bOverrideAngularDamping=true
    bUseGravityOverride=true
    gravityScale=0.35
    simulationSpace=RootRelative
    alphaBoolBlend {blendOption=Linear}
    boundBone {boneName=dyn_lower_pocket_r}
    componentAppliedLinearAccClamp=<3000 3000 3000>
    componentLinearAccScale=<1 1 1>
    componentLinearVelScale=<1 1 1>
    gravityOverride {z=-980}
    physicsBodyDefinitions [{
      boundBone {boneName=dyn_lower_pocket_r}
      boxExtents=<15 25 25>
      constraintSetup {angularLimitsMax=<0 0 15>, angularLimitsMin=<0 0 -50>, angularTarget=<1 0 0>}
      localJointOffset=<4 0 0>
     }]
   }
   n36 = Anim Dynamics - Bone: dyn_upper_pocket_r!(ComponentPose=@n35, Alpha=1, GravityOverride:Vector="0,0,-980", ExternalForce="0,0,0") -> (Pose) {
    angularDampingOverride=1
    angularSpringConstant=20000
    bAngularSpring=true
    bOverrideAngularDamping=true
    bUseGravityOverride=true
    gravityScale=0.35
    simulationSpace=RootRelative
    alphaBoolBlend {blendOption=Linear}
    boundBone {boneName=dyn_upper_pocket_r}
    componentAppliedLinearAccClamp=<3000 3000 3000>
    componentLinearAccScale=<1 1 1>
    componentLinearVelScale=<1 1 1>
    gravityOverride {z=-980}
    physicsBodyDefinitions [{
      boundBone {boneName=dyn_upper_pocket_r}
      boxExtents=<15 25 25>
      constraintSetup {angularLimitsMax=<0 0 15>, angularLimitsMin=<0 0 -50>, angularTarget=<1 0 0>}
      localJointOffset=<4 0 0>
     }]
   }
   n37 = Anim Dynamics - Bone: dyn_Shoe_Strap_r!(ComponentPose=@n36, Alpha=1, GravityOverride:Vector="0,0,-980", ExternalForce="0,0,0") -> (Pose) {
    angularDampingOverride=1
    angularSpringConstant=1000
    bAngularSpring=true
    bOverrideAngularDamping=true
    bUseGravityOverride=true
    gravityScale=0.35
    simulationSpace=BoneRelative
    alphaBoolBlend {blendOption=Linear}
    boundBone {boneName=dyn_Shoe_Strap_r}
    componentAppliedLinearAccClamp=<3000 3000 3000>
    componentLinearAccScale=<1 1 1>
    componentLinearVelScale=<1 1 1>
    gravityOverride {z=-980}
    physicsBodyDefinitions [{
      boundBone {boneName=dyn_Shoe_Strap_r}
      boxExtents=<9 9 9>
      constraintSetup {angularLimitsMax=<0 30 0>, angularLimitsMin=<0 -30 0>, angularTarget=<1 0 0>}
      localJointOffset=<3.2699 0 -5>
     }]
    relativeSpaceBone {boneName=calf_twist_01_r}
   }
   n38 = Anim Dynamics - Bone: dyn_Shoe_Strap_l!(ComponentPose=@n37, Alpha=1, GravityOverride:Vector="0,0,-980", ExternalForce="0,0,0") -> (Pose) {
    angularDampingOverride=1
    angularSpringConstant=1000
    bAngularSpring=true
    bOverrideAngularDamping=true
    bUseGravityOverride=true
    gravityScale=0.35
    simulationSpace=BoneRelative
    alphaBoolBlend {blendOption=Linear}
    boundBone {boneName=dyn_Shoe_Strap_l}
    componentAppliedLinearAccClamp=<3000 3000 3000>
    componentLinearAccScale=<1 1 1>
    componentLinearVelScale=<1 1 1>
    gravityOverride {z=-980}
    physicsBodyDefinitions [{
      boundBone {boneName=dyn_Shoe_Strap_l}
      boxExtents=<9 9 9>
      constraintSetup {angularLimitsMax=<0 30 0>, angularLimitsMin=<0 -30 0>, angularTarget=<1 0 0>}
      localJointOffset=<3.2699 0 -5>
     }]
    relativeSpaceBone {boneName=calf_twist_01_l}
   }
   n39 = Anim Dynamics - Bone: hood_back_c!(ComponentPose=@n38, Alpha=1, GravityOverride:Vector="0,0,-980", ExternalForce="0,0,0") -> (Pose) {
    angularDampingOverride=1
    angularSpringConstant=1000
    bAngularSpring=true
    bOverrideAngularDamping=true
    bUseGravityOverride=true
    gravityScale=0.35
    simulationSpace=BoneRelative
    alphaBoolBlend {blendOption=Linear}
    boundBone {boneName=hood_back_c}
    componentAppliedLinearAccClamp=<3000 3000 3000>
    componentLinearAccScale=<1 1 1>
    componentLinearVelScale=<1 1 1>
    gravityOverride {z=-980}
    physicsBodyDefinitions [{
      boundBone {boneName=hood_back_c}
      boxExtents=<15 15 30>
      constraintSetup {angularLimitsMax=<15 0 50>, angularLimitsMin=<-15 0 -30>, angularTarget=<1 0 0>}
      localJointOffset=<-4 -5 0>
     }]
    relativeSpaceBone {boneName=head}
   }
   Output Pose(Result=Component To Local(ComponentPose=@n39)) {layerGroup=None}
  ]}
 EventGraph {nodes [
   Event BlueprintUpdateAnimation()
   > n1 = Gravity Override State Function DEPRECATED(
    CharacterPartAnimInstance=Self-Reference()
    GameTimeInSeconds=GetGameTimeInSeconds()
    GravityOverrideParams=Get Gravity Override Params()
    RBAN Params=Get RBAN Params()
   )
   > Set Sim Space Settings(Sim Space Settings:Sim Space Settings Structure=Break RBANSetting Struct!(RBANSettingStruct:"RBANSetting Struct Structure (by ref)"=@n1.RBAN SettingsOut) -> (
     ComponentLinearAccScale_28_#1:Vector=~
     ComponentLinearVelScale_31_#2:Vector=~
     SimSpaceSettings_34_#3:Sim Space Settings Structure
    ) {is=title K2Node_BreakStruct}.SimSpaceSettings_34_#3) -> (Output_Get:Sim Space Settings Structure)
   > Set GravityOverride(GravityOverride=@n1)
   > n2 = Branch(Condition=Get bIsInFrontEnd())
   > Set Hoodie Laces Alpha(Hoodie Laces Alpha:float=0.3) -> (Output_Get:float)
   > n3 = Branch(Condition=Get bIsInFrontEnd())
   > Set SimlationSpace(SimlationSpace:AnimPhysSimSpaceType Enum=World) -> (Output_Get:AnimPhysSimSpaceType Enum)
   > Set AlphaFE(AlphaFE:double=0.3) -> (Output_Get:double)
   Set SimlationSpace(execute=@n3.else)
   > Set AlphaFE(AlphaFE=1)
   Set Hoodie Laces Alpha(execute=@n2.else, Hoodie Laces Alpha=1)
   Event OnCharacterCustomizationLoaded()
   > n4 = Cast To FortPlayerPawn(Object=GetOwningActor())
   > n5 = GetVariantChannelActiveIndexForItemActor!(
    execute
    Item:Athena Cosmetic Item Definition Object Reference=/BRCosmetics/Athena/Items/Cosmetics/Characters/Character_CyberMittHomer
    Actor:Actor Object Reference=@n4
    Channel:Gameplay Tag Structure=(TagName="Cosmetics.Variant.Channel.Progressive")
    VariantList:Array of Gameplay Tag Structures=Make Array!("[0]"=(TagName="Cosmetics.Variant.Property.Stage1"), "[1]":Gameplay Tag Structure=(TagName="Cosmetics.Variant.Property.Stage2")) -> (Array)
    bErrorIfChannelMissing:bool=false
    bErrorIfTagNotInArray:bool=false
   ) -> (then, Index:int=0, bIndexIsValid:bool, bFoundChannel:bool)
   > n6 = Branch(Condition="Equal (Integer)"(A=@n5))
   > Set isUsingHelmet(isUsingHelmet:bool=true) -> (Output_Get:bool)
   Set isUsingHelmet(execute=@n6.else, isUsingHelmet=false)
   "Equal (Integer)"(A=@n5, B=1)
  ]}
]
variables [
 AlphaFE {type=number}
 Gravity Override Params {type=GravityOverrideParamsStruct}
 GravityOverride {type=Vector}
 Hoodie Laces Alpha {type=number}
 isUsingHelmet {type=boolean}
 RBAN Params {
  type=RBANParamsStruct
  default {
   dBNO {simSpaceSettings {
     maxAngularAcceleration=1000
     maxAngularVelocity=3
     maxLinearAcceleration=300
     maxLinearVelocity=700
     velocityScaleZ=0
     externalLinearDragV=<0 0.5 0>
    }}
   dBNO_Carried {simSpaceSettings {
     maxAngularAcceleration=1000
     maxAngularVelocity=3
     maxLinearAcceleration=300
     maxLinearVelocity=700
     velocityScaleZ=0
     externalLinearDragV=<0 0.5 0>
    }}
   emoteOrMelee {simSpaceSettings {
     maxAngularAcceleration=1000
     maxAngularVelocity=3
     maxLinearAcceleration=300
     maxLinearVelocity=700
     velocityScaleZ=0
     externalLinearDragV=<0 0.5 0>
    }}
   front_End {simSpaceSettings {
     maxAngularAcceleration=1000
     maxAngularVelocity=3
     maxLinearAcceleration=300
     maxLinearVelocity=700
     velocityScaleZ=0
     externalLinearDragV=<0 0.5 0>
    }}
   onGround_Crouch_Moving {simSpaceSettings {
     maxAngularAcceleration=1000
     maxAngularVelocity=3
     maxLinearAcceleration=300
     maxLinearVelocity=700
     velocityScaleZ=0
     externalLinearDragV=<0 0.5 0>
    }}
   onGround_Falling {simSpaceSettings {
     maxAngularAcceleration=1000
     maxAngularVelocity=3
     maxLinearAcceleration=300
     maxLinearVelocity=700
     velocityScaleZ=0
     externalLinearDragV=<0 0.5 0>
    }}
   onGround_Moving {simSpaceSettings {
     maxAngularAcceleration=1000
     maxAngularVelocity=3
     maxLinearAcceleration=300
     maxLinearVelocity=700
     velocityScaleZ=0
     externalLinearDragV=<0 0.5 0>
    }}
   onGround_Standing {simSpaceSettings {
     maxAngularAcceleration=1000
     maxAngularVelocity=3
     maxLinearAcceleration=300
     maxLinearVelocity=700
     velocityScaleZ=0
     externalLinearDragV=<0 0.5 0>
    }}
   rail_Grinding {simSpaceSettings {
     maxAngularAcceleration=1000
     maxAngularVelocity=3
     maxLinearAcceleration=300
     maxLinearVelocity=700
     velocityScaleZ=0
     worldAlpha=0.1
     externalLinearDragV=<0 0.5 0>
    }}
   riding_Moving {simSpaceSettings {
     maxAngularAcceleration=1000
     maxAngularVelocity=3
     maxLinearAcceleration=300
     maxLinearVelocity=700
     velocityScaleZ=0
     externalLinearDragV=<0 0.5 0>
    }}
   riding_Moving_Targeting {simSpaceSettings {
     maxAngularAcceleration=1000
     maxAngularVelocity=3
     maxLinearAcceleration=300
     maxLinearVelocity=700
     velocityScaleZ=0
     externalLinearDragV=<0 0.5 0>
    }}
   riding_Standing {simSpaceSettings {
     maxAngularAcceleration=1000
     maxAngularVelocity=3
     maxLinearAcceleration=300
     maxLinearVelocity=700
     velocityScaleZ=0
     externalLinearDragV=<0 0.5 0>
    }}
   skydive_DivingDown {simSpaceSettings {
     maxAngularAcceleration=1000
     maxAngularVelocity=3
     maxLinearAcceleration=300
     maxLinearVelocity=700
     velocityScaleZ=0
     externalLinearDragV=<0 0.5 0>
    }}
   skydive_FloatingDown {
    componentLinearAccScale=<0 0.5 0>
    componentLinearVelScale=<0 0.5 0>
    simSpaceSettings {
     maxAngularAcceleration=1000
     maxAngularVelocity=3
     maxLinearAcceleration=300
     maxLinearVelocity=700
     velocityScaleZ=0
     externalLinearDragV=<0 0.5 0>
    }
   }
   skydive_GoingUp {simSpaceSettings {
     maxAngularAcceleration=1000
     maxAngularVelocity=3
     maxLinearAcceleration=300
     maxLinearVelocity=700
     velocityScaleZ=0
     externalLinearDragV=<0 0.5 0>
    }}
   skydive_Parachuting {
    componentLinearAccScale=<0 0.1 0.3>
    componentLinearVelScale=<0 0.1 0.3>
    simSpaceSettings {
     maxAngularAcceleration=1000
     maxAngularVelocity=3
     maxLinearAcceleration=300
     maxLinearVelocity=700
     velocityScaleZ=0
     externalLinearDragV=<0 0.5 0>
    }
   }
   skydive_Surfing {simSpaceSettings {
     maxAngularAcceleration=1000
     maxAngularVelocity=3
     maxLinearAcceleration=300
     maxLinearVelocity=700
     velocityScaleZ=0
     externalLinearDragV=<0 0.5 0>
    }}
   swimming {simSpaceSettings {
     maxAngularAcceleration=1000
     maxAngularVelocity=3
     maxLinearAcceleration=300
     maxLinearVelocity=700
     velocityScaleZ=0
     externalLinearDragV=<0 0.5 0>
    }}
   tactical_Sprint {simSpaceSettings {
     maxAngularAcceleration=1000
     maxAngularVelocity=3
     maxLinearAcceleration=300
     maxLinearVelocity=700
     velocityScaleZ=0
     externalLinearDragV=<0 0.5 0>
    }}
   ziplining {simSpaceSettings {
     maxAngularAcceleration=1000
     maxAngularVelocity=3
     maxLinearAcceleration=300
     maxLinearVelocity=700
     velocityScaleZ=0
     externalLinearDragV=<0 0.5 0>
    }}
  }
 }
 Sim Space Settings {
  type=SimSpaceSettings
  default {
   dampingAlpha=1
   maxAngularAcceleration=10000
   maxAngularVelocity=10000
   maxLinearAcceleration=10000
   maxLinearVelocity=10000
   velocityScaleZ=1
   worldAlpha=0
   externalLinearDragV=<0 0 0>
  }
 }
 SimlationSpace {type=AnimPhysSimSpaceType}
]
