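# Physics asset for the left and right earrings on the M_MED_CyberMitt_Homer male character's face accessory mesh.
# Each side has a small capsule body on the base earring bone and a larger, offset capsule on the dyn earring bone.
# A constraint joins each dyn bone to its base, with a wider Swing2 limit and near-zero Swing1 and twist limits, favoring single-axis swing.
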PhysicsAsset /BRCosmetics/Characters/Player/Male/Medium/Bodies/M_MED_CyberMitt_Homer/Meshes/Parts/M_MED_CyberMitt_Homer_FaceAcc_Physics
bodies {
 base_earring_l [{shapeType=Capsule, length=0.505, radius=0.505, shapeName=base_earring_l_capsule, rotation=<rot 0 -90 0>}]
 base_earring_r [{shapeType=Capsule, length=0.505, radius=0.505, shapeName=base_earring_r_capsule, rotation=<rot 0 -90 0>}]
 dyn_earring_l [{shapeType=Capsule, length=2, radius=1, shapeName=dyn_earring_l_capsule, center=<1 0 0>, rotation=<rot -90 0 0>}]
 dyn_earring_r [{shapeType=Capsule, length=2, radius=1, shapeName=dyn_earring_r_capsule, center=<1 0 0>, rotation=<rot -90 0 0>}]
}
constraints [
 {bone1Name=dyn_earring_l, bone2Name=base_earring_l, swing1LimitDegrees=1, swing2LimitDegrees=20, twistLimitDegrees=1}
 {bone1Name=dyn_earring_r, bone2Name=base_earring_r, swing1LimitDegrees=1, swing2LimitDegrees=20, twistLimitDegrees=1}
]
